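# Physics asset for the HeroicTale glider chute cosmetic, giving its skeletal mesh a simplified collision and joint setup.
# Each major bone has one capsule body, covering the parachute, top disc, frames, handles, wings, missiles, canisters, holster, boxes and one top rope.
# Constraints form a tree around Frame, with missiles on wings, the right canister on the holster, and the parachute on Disc_Top.
# All joints share identical swing and twist limit angles.
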PhysicsAsset /BRCosmetics/Gadgets/Assets/VinderTech_GliderChute/Glider_HeroicTale/Meshes/HeroicTale_Glider_PhysicsAsset
bodies {
 Box_L [{shapeType=Capsule, length=16.2716, radius=15.8695, center=<0.2596 4.7662 -10.8412>, rotation=<rot 34.1276 -90 -125.1085>}]
 Box_R [{shapeType=Capsule, length=16.2716, radius=15.8695, center=<0.1904 4.7662 -10.8412>, rotation=<rot 34.1276 -90 -54.8915>}]
 Cannister_L [{shapeType=Capsule, length=59.8288, radius=8.7821, center=<1.1821 34.6572 -0.179>, rotation=<rot 0 -1.0428 -89.5725>}]
 Cannister_R [{shapeType=Capsule, length=59.8097, radius=13.0166, center=<2.1983 34.7064 0.2786>, rotation=<rot 0 1.9556 -88.9393>}]
 Disc_Top [{shapeType=Capsule, length=14.0576, radius=14.0531, center=<0.2279 4.1062 -3.7147>, rotation=<rot 0 63.211 -89.9163>}]
 Frame [{shapeType=Capsule, length=75.2282, radius=70.7116, center=<1.5882 -19.0704 3.7068>, rotation=<rot 2.4289 88.7414 -90>}]
 Handles [{shapeType=Capsule, length=49.2128, radius=31.762, center=<-3.5888 25.0249 -11.4447>, rotation=<rot 0 81.7555 -84.5087>}]
 Holster_R [{shapeType=Capsule, length=40.2032, radius=12.8794, center=<-7.5957 21.9262 -11.4458>, rotation=<rot 0 179.0292 -89.7828>}]
 Missile_L [{shapeType=Capsule, length=38.4148, radius=3.7367, center=<0.3599 0.8301 0.0799>, rotation=<rot 0 -0.3706 -89.5119>}]
 Missile_R [{shapeType=Capsule, length=38.4148, radius=3.7367, center=<-0.0632 0.7576 0.0544>, rotation=<rot 0 179.6294 -89.5119>}]
 Rope_Top_B_R [{shapeType=Capsule, length=27.0766, radius=21.4599, center=<-4.0683 1.7553 -12.1317>, rotation=<rot 76.6672 -12.6914 76.9689>}]
 SubFrame [{shapeType=Capsule, length=76.0844, radius=68.9653, center=<-0.1652 -11.7775 -5.543>, rotation=<rot 0 -151.9527 -87.3847>}]
 Wing_L [{shapeType=Capsule, length=53.3458, radius=26.9404, center=<-22.3385 -0.5945 -5.3524>, rotation=<rot -0.0864 -90 3.3358>}]
 Wing_R [{shapeType=Capsule, length=43.2632, radius=27.3376, center=<-22.4872 0.1415 0.185>, rotation=<rot -0.2043 -90 -4.3814>}]
 parachute [{shapeType=Capsule, length=80.5424, radius=76.8832, center=<10.6903 -10.6359 -31.5903>, rotation=<rot 0 64.2216 -89.8125>}]
}
constraints [
 {bone1Name=Frame, bone2Name=Disc_Top, swing1LimitDegrees=45, swing2LimitDegrees=45, twistLimitDegrees=45}
 {bone1Name=SubFrame, bone2Name=Frame, swing1LimitDegrees=45, swing2LimitDegrees=45, twistLimitDegrees=45}
 {bone1Name=Handles, bone2Name=SubFrame, swing1LimitDegrees=45, swing2LimitDegrees=45, twistLimitDegrees=45}
 {bone1Name=Wing_L, bone2Name=Frame, swing1LimitDegrees=45, swing2LimitDegrees=45, twistLimitDegrees=45}
 {bone1Name=Missile_L, bone2Name=Wing_L, swing1LimitDegrees=45, swing2LimitDegrees=45, twistLimitDegrees=45}
 {bone1Name=Wing_R, bone2Name=Frame, swing1LimitDegrees=45, swing2LimitDegrees=45, twistLimitDegrees=45}
 {bone1Name=Missile_R, bone2Name=Wing_R, swing1LimitDegrees=45, swing2LimitDegrees=45, twistLimitDegrees=45}
 {bone1Name=Cannister_L, bone2Name=Frame, swing1LimitDegrees=45, swing2LimitDegrees=45, twistLimitDegrees=45}
 {bone1Name=Holster_R, bone2Name=Frame, swing1LimitDegrees=45, swing2LimitDegrees=45, twistLimitDegrees=45}
 {bone1Name=Cannister_R, bone2Name=Holster_R, swing1LimitDegrees=45, swing2LimitDegrees=45, twistLimitDegrees=45}
 {bone1Name=Box_R, bone2Name=Frame, swing1LimitDegrees=45, swing2LimitDegrees=45, twistLimitDegrees=45}
 {bone1Name=Box_L, bone2Name=Frame, swing1LimitDegrees=45, swing2LimitDegrees=45, twistLimitDegrees=45}
 {bone1Name=parachute, bone2Name=Disc_Top, swing1LimitDegrees=45, swing2LimitDegrees=45, twistLimitDegrees=45}
 {bone1Name=Rope_Top_B_R, bone2Name=parachute, swing1LimitDegrees=45, swing2LimitDegrees=45, twistLimitDegrees=45}
]
